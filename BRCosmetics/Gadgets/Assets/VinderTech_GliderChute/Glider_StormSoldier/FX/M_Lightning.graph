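# Additive, unlit, two-sided lightning material for the Storm Soldier glider FX.
# Panning noise distorts a gradient lookup that drives a curve-atlas color, squared and scaled by Ani_Intensity into emissive adjustments.
# A separate offset chain combines camera-relative cross-product widening (Ani_Size) with panning-noise jitter (Ani_Offset, Noise), masked by a ramp.
# Lightning-sheet samples and several color and opacity nodes feed neither chain.
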Material /BRCosmetics/Gadgets/Assets/VinderTech_GliderChute/Glider_StormSoldier/FX/M_Lightning
graph {
 properties {
  bUsedWithMeshDeformer=true
  bUsedWithSkeletalMesh=true
  blendMode=BLEND_Additive
  pixelDepthOffsetMode=PDOM_Legacy
  shadingModel=MSM_Unlit
  twoSided=true
  usedShadingModels=Unlit
 }
 nodes [
  n1 = MaterialFunctionCall() {materialFunction=&Swizzle}
  n2 = Multiply(A=@n1.YX, B=Constant2Vector() {g=0.1, r=0.1})
  n3 = TextureSample(UVs=Add(A=@n2, B=Constant2Vector() {r=0.1})) {texture=/Game/Athena/HUD/Materials/T-UI-LightningSheet}
  Add(A=TextureSample(UVs=@n2) {texture=&T-UI-LightningSheet}.R, B=@n3.R)
  Add(A=@n1.YX)
  n4 = CrossProduct(
   A=Transform(None=Constant3Vector() {constant=<rgba 1 0 0 0>})
   B=Transform(None=Constant3Vector() {constant=<rgba 0 0 1 0>}) {transformSourceType=TRANSFORMSOURCE_Camera}
  )
  n5 = ComponentMask(None=Subtract(A=TextureCoordinate() {uTiling=2, vTiling=2})) {g=true}
  n6 = ScalarParameter() {parameterName=Ani_Size}
  n7 = TextureCoordinate() {coordinateIndex=1, uTiling=0.03}
  n8 = Subtract(A=TextureSample(UVs=Panner(Coordinate=@n7) {speedX=0.25, speedY=0.005}) {samplerType=SAMPLERTYPE_LinearColor, texture=/Game/Effects/Fort_Effects/Textures/Tile/LinerNoise64x64}.RGB) {constB=0.33}
  n9 = Add(B=ScalarParameter() {parameterName=Ani_Offset}) {constA=10}
  n10 = TextureCoordinate()
  n11 = TextureSample(UVs=MaterialFunctionCall(XY=@n10) {materialFunction=&Swizzle}.YX) {texture=/Game/Effects/Fort_Effects/Textures/Ramps/T_Power_Source_Sphere_Ramp}
  n12 = ScalarParameter() {defaultValue=1, parameterName=Noise}
  Add(
   A=Multiply(A=Multiply(A=@n4, B=@n5), B=Add(B=@n6) {constA=5}) {constB=6}
   B=Multiply(A=Multiply(A=@n8, B=@n9) {constB=32}, B=Multiply(A=@n11.G, B=@n12))
  )
  CameraVectorWS()
  Constant3Vector() {constant=<rgba 128 0 60.4464 1>}
  CurveAtlasRowParameter(CurveTime=ComponentMask(None=@n10) {r=true}) {
   atlas=&AthenaCurveAtlas01
   curve=/Game/Athena/LinearColorCurves/Curve_Ramps2
   defaultValue=98
   parameterName=Param2
  }
  n13 = TextureCoordinate()
  n14 = TextureCoordinate() {uTiling=0.1}
  n15 = ComponentMask(None=TextureSample(UVs=Panner(Coordinate=@n14) {speedX=1, speedY=0.5}) {samplerType=SAMPLERTYPE_LinearColor, texture=&T_NoisePack_LinearColor}.RGB) {g=true, r=true}
  n16 = Multiply(A=Multiply(A=Subtract(A=@n15) {constB=0.4}, B=TextureSample() {texture=/Game/Global/Textures/Default/Masks_and_Gradients/T_Water_Top_Ramp}.R)) {constB=0.6}
  n17 = CurveAtlasRowParameter(CurveTime=TextureSample(UVs=Add(A=@n13, B=@n16)) {texture=&T_Water_Top_Ramp}.R) {
   atlas=&AthenaCurveAtlas01
   curve=/Game/Athena/LinearColorCurves/Curve_Ramps
   defaultValue=95
   parameterName=Param2
  }
  n18 = Multiply(A=@n17, B=@n17)
  n19 = Clamp(None=Add(A=ScalarParameter() {parameterName=Ani_Intensity})) {clampMode=CMODE_ClampMin}
  MaterialFunctionCall(Emissive=Multiply(A=Multiply(A=@n18, B=@n19)) {constB=2}) {materialFunction=&MF_FX_EmissiveAdjustments}
  Multiply(A=@n3.R) {constB=0.1}
  ScalarParameter() {defaultValue=1, parameterName=Opacity}
  VectorParameter() {defaultValue=<rgba 2 0 0.9445 1>}
  VertexNormalWS()
 ]
}
